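# Unlit masked electric arc material for Niagara ribbons in the Darkmatter dealership set.
# World position offset comes from time-driven sine and cosine waves over distance from actor position, tuned by the WPO Values vector.
# Color is a hard central band across UV V, tinted by particle color and passed through emissive adjustments.
# Opacity thresholds panning cloud noise sampled with zero UV tiling, so the arc flickers uniformly rather than showing a spatial pattern.
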Material /DealershipSharedLibrary/Assets/Sets/Darkmatter/Material/M_DLR_ElectricArc
graph {
 properties {bAutomaticallySetUsageInEditor=false, bUsedWithNiagaraRibbons=true, blendMode=BLEND_Masked, shadingModel=MSM_Unlit, usedShadingModels=Unlit}
 nodes [
  n1 = VectorParameter() {parameterName=WPO Values, defaultValue=<rgba 0 0 0 1>}
  n2 = Multiply(A=@n1.G, B=OneMinus(None=Time()))
  n3 = Subtract(A=ActorPositionWS(), B=WorldPosition().XYZ)
  n4 = MaterialFunctionCall(Color=Divide(A=Abs(None=@n3), B=@n1.R)) {materialFunction=&SplitComponents}
  n5 = Add(A=Add(A=@n4.R, B=@n4.R), B=@n4.B)
  n6 = Add(A=@n2, B=@n5)
  n7 = Multiply(A=Cosine(None=@n6) {period=1.25}, B=Sine(None=@n6))
  n8 = MaterialFunctionCall(R=@n7, G=Sine(None=Add(A=@n5, B=@n2)), B=@n7) {materialFunction=&AppendMany}
  NamedRerouteDeclaration(None=Transform(None=Multiply(A=@n8.RGB, B=@n1.B)) {transformSourceType=TRANSFORMSOURCE_Local}) {name=WPO, nodeColor=<rgba 1 0.0703 0 1>}
  n9 = ComponentMask(None=TextureCoordinate()) {g=true}
  n10 = Multiply(A=@n9, B=OneMinus(None=@n9))
  n11 = Saturate(None=Ceil(None=Subtract(A=@n10) {constB=0.2}))
  n12 = ParticleColor()
  NamedRerouteDeclaration(None=MaterialFunctionCall(Emissive=Multiply(A=@n11, B=@n12.RGB)) {materialFunction=&MF_FX_EmissiveAdjustments}.Result) {name=Color, nodeColor=<rgba 0 0.25 1 1>}
  n13 = TextureCoordinate() {coordinateIndex=1, uTiling=0, vTiling=0}
  n14 = Multiply(A=Time(), B=ScalarParameter() {defaultValue=1, parameterName=Tile})
  n15 = Subtract(A=TextureSample(UVs=Panner(Coordinate=@n13, Time=@n14) {speedX=2, speedY=3}) {texture=/DealershipSharedLibrary/FX/Textures/Noise/T_VCVFX_CloudNoise}.RGB) {constB=0.4}
  NamedRerouteDeclaration(None=Saturate(None=Multiply(A=@n15) {constB=55})) {name=Opacity, nodeColor=<rgba 0.5703 1 0 1>}
  NamedRerouteUsage()
  NamedRerouteUsage()
  NamedRerouteUsage()
  VertexColor()
 ]
}
